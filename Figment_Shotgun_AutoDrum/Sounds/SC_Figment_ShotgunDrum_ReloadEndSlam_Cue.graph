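# Sound cue for the end-of-reload slam on the Figment drum auto shotgun.
# A mixer layers the shotgun drum's reload-end wave at reduced volume under a full-volume magazine-insert slam borrowed from the PDW.
# The cue slightly raises overall pitch and volume, and plays in the weapon foley sound class with the shared weapon reload attenuation.
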SoundCue /Figment_Shotgun_AutoDrum/Sounds/SC_Figment_ShotgunDrum_ReloadEndSlam_Cue
root=Mixer_0
nodes [
 {id=Mixer_0, children [WavePlayer_0,WavePlayer_1], properties {inputVolume [0.3,1]}}
 {id=WavePlayer_0, properties {soundWaveAssetPtr=&/ShotgunDrum_ReloadEnd_04}}
 {id=WavePlayer_1, properties {soundWaveAssetPtr=&/PDW_Reload_InsertSlam_01}}
]
properties {
 attenuationSettings=&Fort_Weapon_Reload_Att
 maxDistance=2250
 pitchMultiplier=1.05
 soundClassObject=&Fort_WeaponFoley
 volumeMultiplier=1.2
}
