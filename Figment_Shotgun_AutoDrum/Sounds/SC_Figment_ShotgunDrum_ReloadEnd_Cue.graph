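# Sound cue for the reload-end foley of the Figment auto drum shotgun.
# It plays a single wave, a drum reload insert sound, with no randomization or mixing, as weapon foley with boosted volume and the standard weapon reload attenuation.
# Its overridden concurrency limits instances per owner to a small count and stops the oldest when exceeded.
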SoundCue /Figment_Shotgun_AutoDrum/Sounds/SC_Figment_ShotgunDrum_ReloadEnd_Cue
root=WavePlayer_1
nodes [{id=WavePlayer_1, properties {soundWaveAssetPtr=&/ShotgunDrum_ReloadInsert_01}}]
properties {
 attenuationSettings=&Fort_Weapon_Reload_Att
 bOverrideConcurrency=true
 maxDistance=2250
 soundClassObject=&Fort_WeaponFoley
 volumeMultiplier=1.4
 attenuationOverrides {falloffDistance=1600}
 concurrencyOverrides {bLimitToOwner=true, maxCount=2, resolutionRule=StopOldest}
}
