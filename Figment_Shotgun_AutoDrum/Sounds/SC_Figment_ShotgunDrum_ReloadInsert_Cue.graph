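# Sound cue for the magazine-insert step of the Figment drum shotgun reload.
# A single HiInfantry reload-start wave plays pitched down by a fixed modulator after a short fixed delay.
# It uses weapon foley sound class and reload attenuation, with boosted volume and a concurrency override that prevents new instances when the limit is reached.
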SoundCue /Figment_Shotgun_AutoDrum/Sounds/SC_Figment_ShotgunDrum_ReloadInsert_Cue
root=Delay_0
nodes [
 {id=Delay_0, children [Modulator_0], properties {delayMax=0.15, delayMin=0.15}}
 {id=Modulator_0, children [WavePlayer_0], properties {pitchMax=0.9, pitchMin=0.9}}
 {id=WavePlayer_0, properties {soundWaveAssetPtr=&/HiInfantry_ReloadStart_01}}
]
properties {
 attenuationSettings=&Fort_Weapon_Reload_Att
 bOverrideConcurrency=true
 maxDistance=2250
 soundClassObject=&Fort_WeaponFoley
 volumeMultiplier=2
 attenuationOverrides {falloffDistance=1600}
 concurrencyOverrides {resolutionRule=PreventNew}
}
