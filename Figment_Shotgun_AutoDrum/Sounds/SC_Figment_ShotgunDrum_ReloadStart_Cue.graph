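# Sound cue for the start of the reload on the Figment auto drum shotgun.
# A single wave player feeds the root with the drum reload start sound, slightly lowered in pitch and raised in volume.
# It uses shared weapon reload attenuation in the weapon foley sound class, with a concurrency override that blocks new instances once its limit is reached.
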SoundCue /Figment_Shotgun_AutoDrum/Sounds/SC_Figment_ShotgunDrum_ReloadStart_Cue
root=WavePlayer_0
nodes [{id=WavePlayer_0, properties {soundWaveAssetPtr=&/ShotgunDrum_ReloadStart_01}}]
properties {
 attenuationSettings=&Fort_Weapon_Reload_Att
 bOverrideConcurrency=true
 maxDistance=2250
 pitchMultiplier=0.95
 soundClassObject=&Fort_WeaponFoley
 volumeMultiplier=1.1
 attenuationOverrides {falloffDistance=1600}
 concurrencyOverrides {resolutionRule=PreventNew}
}
